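; Non-participant pawn Blueprint for MotherGround that tracks player proximity for a dance quest.
; A quest-tracker collision sphere drives it, acting only on the server authority.
; On overlap begin, the overlapping Fort player pawn's ability system applies the DanceQuestTracker gameplay effect to itself.
; On overlap end, effects granting the Athena.Quests.DanceGardener tag are removed from that pawn.
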(event EventBeginPlay
  (|CallParentFunction) ; This node is disabled and will not be called.\nDrag off pins to build functionality.
)

(event Collision|EventActorBeginOverlap
  (params (OtherActor :type Actor%20Object%20Reference))
  (|CallParentFunction OtherActor) ; This node is disabled and will not be called.\nDrag off pins to build functionality.
)

(event EventTick
  (params (DeltaSeconds :type Float%20%28single-precision%29))
  (|CallParentFunction DeltaSeconds) ; This node is disabled and will not be called.\nDrag off pins to build functionality.
)

(event OnComponentBeginOverlap(SphereCollision_QuestTracker)
  (params (OverlappedComponent :type Primitive%20Component%20Object%20Reference) (OtherActor :type Actor%20Object%20Reference) (OtherComp :type Primitive%20Component%20Object%20Reference) (OtherBodyIndex :type Integer) (bFromSweep :type Boolean) (SweepResult :type Hit%20Result%20Structure%20%28by%20ref%29))
  (Networking|SwitchHasAuthority
    (:Authority
      (bind _asfort_player_pawn (Utilities|Casting|CastToFortPlayerPawn OtherActor))
      (GameplayEffects|ApplyGameplayEffectToTarget :self (Ability|GetAbilitySystemComponent _asfort_player_pawn) :GameplayEffectClass "/MotherGround/Abilities/GE_MotherGround_DanceQuestTracker.GE_MotherGround_DanceQuestTracker_C" :Target (Ability|GetAbilitySystemComponent _asfort_player_pawn) :Level 0.0))))

(event OnComponentEndOverlap(SphereCollision_QuestTracker)
  (params (OverlappedComponent :type Primitive%20Component%20Object%20Reference) (OtherActor :type Actor%20Object%20Reference) (OtherComp :type Primitive%20Component%20Object%20Reference) (OtherBodyIndex :type Integer))
  (Networking|SwitchHasAuthority
    (:Authority
      (bind _asfort_player_pawn (Utilities|Casting|CastToFortPlayerPawn OtherActor))
      (GameplayEffects|RemoveActiveEffectswithGrantedTags (Ability|GetAbilitySystemComponent _asfort_player_pawn) "(GameplayTags=((TagName=\"Athena.Quests.DanceGardener\")))"))))

(fn ConstructionScript
  (|CallParentFunction))
